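# Control Rig for the SableHorizon female medium character's hair and face accessory part that offsets hair bones based on head pose.
# A SphericalPoseReader on the head bone outputs a falloff weight from an asymmetric active region.
# That weight drives two chained ModifyTransforms: one rotates and shifts short back-hair base bones, the other rotates the right side-hair base.
# The offsets sit on base bones, so child dyn_ bones inherit them.
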ControlRigBlueprint /BRCosmetics/Characters/Player/Female/Medium/Bodies/F_MED_SableHorizon_Hair/Meshes/Parts/F_MED_SableHorizon_Hair_FaceAcc_CtrlRig
graphs [
 RigVMFunctionLibrary {links [], nodes []}
 RigVMModel {nodes [
   RigUnit_BeginExecution = &RigUnit_BeginExecution()
   RigVMFunction_Sequence = &RigVMFunction_Sequence(ExecuteContext=@RigUnit_BeginExecution.ExecutePin)
   n1 = &RigUnit_SphericalPoseReader(
    ActiveRegionScaleFactors=(PositiveWidth=1,NegativeWidth=0.2,PositiveHeight=1.5,NegativeHeight=1.5)
    ActiveRegionSize=0.1
    Debug=(bDrawDebug=false,bDraw2D=false,bDrawLocalAxes=false,DebugScale=25,DebugSegments=20,DebugThickness=0.25)
    DriverAxis=(X=0,Y=1,Z=0)
    DriverItem=(Type=Bone,Name="head")
    FalloffRegionScaleFactors=(PositiveWidth=3,NegativeWidth=1.8,PositiveHeight=2.5,NegativeHeight=2.5)
    FalloffSize=0.2
    FlipHeightScaling=False
    FlipWidthScaling=False
    OptionalParentItem=(Type=Bone,Name="None")
    RotationOffset=(X=90,Y=0,Z=-55)
    ExecutePin=@RigVMFunction_Sequence.A
   )
   > &RigUnit_ModifyTransforms(
    ItemToModify="((Item=(Type=Bone,Name=\"base_hair_short_bk_c\"),Transform=(Rotation=(X=0,Y=-0.2588,Z=0,W=0.9659),Translation=(X=-5,Y=0,Z=0),Scale3D=(X=1,Y=1,Z=1))),(Item=(Type=Bone,Name=\"base_hair_short_bk_l\"),Transform=(Rotation=(X=0,Y=-0.2588,Z=0,W=0.9659),Translation=(X ...(truncated)"
    WeightMinimum=0
    Weight=@n1.OutputParam
   )
   > &RigUnit_ModifyTransforms(
    ItemToModify=((Item=(Type=Bone,Name="base_hair_side_r"),Transform=(Rotation=(X=0,Y=0,Z=0.3827,W=0.9239),Translation=(X=0,Y=0,Z=0),Scale3D=(X=1,Y=1,Z=1))))
    WeightMinimum=0
    Weight=@n1.OutputParam
   )
  ]}
]
hierarchy [
 root {loc=<0 0 0>, rot=<0 0 0 1>}
 pelvis {parent=root, loc=<0 -0.5301 98.0571>, rot=<0.0127 0.707 -0.0127 -0.707>}
 spine_01 {parent=pelvis, loc=<6.1208 -0.9328 0>, rot=<0 0 0.034 -0.9994>}
 spine_02 {parent=spine_01, loc=<6.5997 -0.2095 0>, rot=<0 0 0.0238 -0.9997>}
 spine_03 {parent=spine_02, loc=<6.4914 -0.1237 0>, rot=<0 0 0.0149 -0.9999>}
 spine_04 {parent=spine_03, loc=<6.9598 0.0794 0>, rot=<0 0 -0.0132 -0.9999>}
 spine_05 {parent=spine_04, loc=<6.9987 0 0>, rot=<0 0 -0.0015 -1>}
 neck_01 {parent=spine_05, loc=<11.3138 0.1529 0>, rot=<0 0 -0.2003 -0.9797>}
 neck_02 {parent=neck_01, loc=<4.5627 0 0>, rot=<0 0 0.0504 -0.9987>}
 head {parent=neck_02, loc=<5.3914 0 0>, rot=<0 0 0.115 -0.9934>}
 base_chain_l {parent=head, loc=<9.1144 1.0949 -8.5066>, rot=<-0.1306 -0.9738 -0.0121 -0.1857>}
 dyn_chain_l {parent=base_chain_l, loc=<0 0 0>, rot=<0 0 0 1>}
 base_chain_r {parent=head, loc=<9.114 1.0949 8.5066>, rot=<0.1857 0.0121 0.9738 0.1306>}
 dyn_chain_r {parent=base_chain_r, loc=<0 0 0>, rot=<0 0 0 -1>}
 base_chain_fr {parent=head, loc=<10.8547 9.3136 -0.0186>, rot=<-0.1084 -0.7005 -0.6974 -0.1058>}
 dyn_chain_fr {parent=base_chain_fr, loc=<0 0 0>, rot=<0 0 0 1>}
 base_chain_bk_r {parent=head, loc=<7.4408 -9.0972 4.1226>, rot=<0.1492 -0.5712 0.7999 -0.1079>}
 dyn_chain_bk_r {parent=base_chain_bk_r, loc=<0 0 0>, rot=<0 0 0 1>}
 base_chain_bk_l {parent=head, loc=<7.4412 -9.0972 -4.1226>, rot=<0.1079 -0.7999 0.5712 -0.1492>}
 dyn_chain_bk_l {parent=base_chain_bk_l, loc=<0 0 0>, rot=<0 0 0 1>}
 base_ear_jewel_l {parent=head, loc=<-1.3019 0.3056 -6.6547>, rot=<0.0118 -0.75 -0.6613 0.0017>}
 dyn_ear_jewel_l {parent=base_ear_jewel_l, loc=<0 0 0>, rot=<0 0 0 1>}
 base_ear_jewel_r {parent=head, loc=<-1.3018 0.3056 6.6547>, rot=<-0.0017 0.6613 0.75 -0.0118>}
 dyn_ear_jewel_r {parent=base_ear_jewel_r, loc=<0 0 0>, rot=<0 0 0 1>}
 base_hair_short_bk_c {parent=head, loc=<8.3818 -7.9828 -0.5102>, rot=<0.1112 -0.7168 0.6795 -0.11>}
 dyn_hair_short_bk_c_01 {parent=base_hair_short_bk_c, loc=<0 0 0>, rot=<0 0 0 1>}
 dyn_hair_short_bk_c_02 {parent=dyn_hair_short_bk_c_01, loc=<7.5769 0.0039 -0.1302>, rot=<0 0.0245 0 -0.9997>}
 dyn_hair_short_bk_c_03 {parent=dyn_hair_short_bk_c_02, loc=<7.0488 -0.0083 -0.188>, rot=<0 -0.0025 0 -1>}
 base_hair_side_r {parent=head, loc=<4.0657 0.2016 9.531>, rot=<0.0032 -0.7507 -0.6605 0.0115>}
 dyn_hair_side_r_1 {parent=base_hair_side_r, loc=<0 0 0>, rot=<0 0 0 1>}
 dyn_hair_side_r_2 {parent=dyn_hair_side_r_1, loc=<11.0838 0 0>, rot=<0 0 0 1>}
 base_hair_side_l {parent=head, loc=<4.3772 -1.2659 -8.3989>, rot=<0.0118 -0.75 -0.6613 0.0017>}
 dyn_hair_side_l_1 {parent=base_hair_side_l, loc=<0 0 0>, rot=<0 0 0 1>}
 dyn_hair_side_l_2 {parent=dyn_hair_side_l_1, loc=<11.0838 0 0>, rot=<0 0 0 1>}
 base_hair_short_bk_r {parent=head, loc=<8.7296 -5.2679 6.0236>, rot=<0.1484 -0.5713 0.8001 -0.1073>}
 dyn_hair_short_bk_r_01 {parent=base_hair_short_bk_r, loc=<0 0 0>, rot=<0 0 0 1>}
 dyn_hair_short_bk_r_02 {parent=dyn_hair_short_bk_r_01, loc=<7.5608 0.0039 -0.0512>, rot=<0 0 0 -1>}
 dyn_hair_short_bk_r_03 {parent=dyn_hair_short_bk_r_02, loc=<6.6492 0.0009 -0.0031>, rot=<0 0 0 -1>}
 base_hair_short_bk_l {parent=head, loc=<8.7296 -5.2679 -6.0236>, rot=<0.1073 -0.8001 0.5713 -0.1484>}
 dyn_hair_short_bk_l_01 {parent=base_hair_short_bk_l, loc=<0 0 0>, rot=<0 0 0 1>}
 dyn_hair_short_bk_l_02 {parent=dyn_hair_short_bk_l_01, loc=<7.5608 -0.0039 -0.0512>, rot=<0 0 0 1>}
 dyn_hair_short_bk_l_03 {parent=dyn_hair_short_bk_l_02, loc=<6.6492 -0.0009 -0.0031>, rot=<0 0 0 1>}
 base_hair_fr_l {parent=head, loc=<8.5679 8.9434 -2.8849>, rot=<0.3277 -0.7497 -0.4545 -0.3522>}
 dyn_hair_fr_l {parent=base_hair_fr_l, loc=<0 0 0>, rot=<0 0 0 1>}
 base_hair_fr_r {parent=head, loc=<8.5681 8.9434 2.8849>, rot=<0.3522 0.4545 0.7497 -0.3277>}
 dyn_hair_fr_r {parent=base_hair_fr_r, loc=<0 0 0>, rot=<0 0 0 1>}
 clavicle_l {parent=spine_05, loc=<5.4831 -0.1053 -3.2946>, rot=<0.0465 -0.7289 -0.009 -0.6829>}
 upperarm_l {parent=clavicle_l, loc=<9.8658 -1.2136 -0.4105>, rot=<0.0164 -0.3245 -0.0269 -0.9454>}
 upperarm_twist_01_l {parent=upperarm_l, loc=<7.2135 -0.1134 0>, rot=<0 0 0 1>}
 clavicle_r {parent=spine_05, loc=<5.4831 -0.1053 3.2946>, rot=<0.7289 0.0465 -0.6829 0.009>}
 upperarm_r {parent=clavicle_r, loc=<-9.8658 1.2136 0.4105>, rot=<0.0164 -0.3245 -0.0269 -0.9454>}
 upperarm_twist_01_r {parent=upperarm_r, loc=<-7.2135 0.1134 0>, rot=<0 0 0 -1>}
]
