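# Physics asset for the SableHorizon female character's hair and face-accessory part, setting up physics for short back hair, side locks and ear jewels.
# Capsules on the head, neck, upper spine, clavicles and upper arms give the body shapes around them.
# Each hair and jewel chain hangs from a head-constrained base bone through angle-limited joints.
# Free-rotation cross-links tie neighboring strands together mid-chain, plus the left side lock to its ear jewel.
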PhysicsAsset /BRCosmetics/Characters/Player/Female/Medium/Bodies/F_MED_SableHorizon_Hair/Meshes/Parts/F_MED_SableHorizon_Hair_FaceAcc_Physics
bodies {
 base_ear_jewel_l [{shapeType=Capsule, length=0.505, radius=0.505, shapeName=base_ear_jewel_l_capsule, rotation=<rot 0 -90 0>}]
 base_ear_jewel_r [{shapeType=Capsule, length=0.505, radius=0.505, shapeName=base_ear_jewel_r_capsule, rotation=<rot 0 -90 0>}]
 base_hair_short_bk_c [{shapeType=Capsule, length=0.505, radius=0.505, shapeName=base_hair_short_bk_c_capsule, rotation=<rot 0 -90 0>}]
 base_hair_short_bk_l [{shapeType=Capsule, length=0.505, radius=0.505, shapeName=base_hair_short_bk_l_capsule, rotation=<rot 0 -90 0>}]
 base_hair_short_bk_r [{shapeType=Capsule, length=0.505, radius=0.505, shapeName=base_hair_short_bk_r_capsule, rotation=<rot 0 -90 0>}]
 base_hair_side_l [{shapeType=Capsule, length=0.505, radius=0.505, shapeName=base_hair_side_l_capsule, rotation=<rot 0 -90 0>}]
 base_hair_side_r [{shapeType=Capsule, length=0.505, radius=0.505, shapeName=base_hair_side_r_capsule, rotation=<rot 0 -90 0>}]
 clavicle_l [{shapeType=Capsule, length=6.899, radius=6.899, shapeName=clavicle_l_capsule, center=<2.6912 0 0>, rotation=<rot -90 0 0>}]
 clavicle_r [{shapeType=Capsule, length=6.899, radius=6.899, shapeName=clavicle_r_capsule, center=<-2.6912 0 0>, rotation=<rot -90 0 0>}]
 dyn_ear_jewel_l [{shapeType=Capsule, length=2, radius=1, center=<2 -1.5 0>, rotation=<rot 90 0 0>}]
 dyn_ear_jewel_r [{shapeType=Capsule, length=2, radius=1, center=<2 1.5 0>, rotation=<rot 90 0 0>}]
 dyn_hair_short_bk_c_01 [{shapeType=Capsule, length=6, radius=4, center=<6 0 1>, rotation=<rot -90 0 90>}]
 dyn_hair_short_bk_c_02 [{shapeType=Capsule, length=6, radius=4, center=<6 0 0>, rotation=<rot -90 0 90>}]
 dyn_hair_short_bk_c_03 [{shapeType=Capsule, length=6, radius=4, center=<6 0 0>, rotation=<rot -90 0 90>}]
 dyn_hair_short_bk_l_01 [{shapeType=Capsule, length=6, radius=4, center=<6 0 2>, rotation=<rot -90 0 90>}]
 dyn_hair_short_bk_l_02 [{shapeType=Capsule, length=6, radius=4, center=<6 0 0>, rotation=<rot -90 0 90>}]
 dyn_hair_short_bk_l_03 [{shapeType=Capsule, length=6, radius=4, center=<6 0 2>, rotation=<rot -90 0 90>}]
 dyn_hair_short_bk_r_01 [{shapeType=Capsule, length=6, radius=4, center=<6 0 4>, rotation=<rot -90 90 0>}]
 dyn_hair_short_bk_r_02 [{shapeType=Capsule, length=6, radius=4, center=<6 0 3>, rotation=<rot -90 0 90>}]
 dyn_hair_short_bk_r_03 [{shapeType=Capsule, length=6, radius=4, center=<6 0 5>, rotation=<rot -90 0 90>}]
 dyn_hair_side_l_1 [{shapeType=Capsule, length=2, radius=3, center=<3 -1 0>, rotation=<rot 0 90 0>}]
 dyn_hair_side_l_2 [{shapeType=Capsule, length=3, radius=2.5, center=<-3 0 0>, rotation=<rot 0 90 0>}]
 dyn_hair_side_r_1 [{shapeType=Capsule, length=2, radius=3, center=<3 0 0>, rotation=<rot 0 90 0>}]
 dyn_hair_side_r_2 [{shapeType=Capsule, length=2, radius=3, center=<-1.5 0 0>, rotation=<rot 0 90 0>}]
 head [{shapeType=Capsule, length=6.974, radius=6, shapeName=head_capsule, rotation=<rot -90 0 0>}]
 neck_01 [{shapeType=Capsule, length=6.4964, radius=5, shapeName=neck_01_capsule, center=<0 2.5433 0>, rotation=<rot -90 0 0>}]
 neck_02 [{shapeType=Capsule, length=6.1956, radius=5, shapeName=neck_02_capsule, rotation=<rot -90 0 0>}]
 spine_03 [{shapeType=Capsule, length=9.0182, radius=9.0182, shapeName=spine_03_capsule, center=<2.2508 -0.9932 0>}]
 spine_04 [{shapeType=Capsule, length=10.2179, radius=10.2179, shapeName=spine_04_capsule, center=<3.0732 -1.0159 0>}]
 spine_05 [{shapeType=Capsule, length=10.5518, radius=10.5518, shapeName=spine_05_capsule, center=<2.4749 0 0>}]
 upperarm_l [{shapeType=Capsule, length=6.5193, radius=6.5193, shapeName=upperarm_l_capsule, center=<3.5311 0 0>, rotation=<rot -90 0 0>}]
 upperarm_r [{shapeType=Capsule, length=6.5193, radius=6.5193, shapeName=upperarm_r_capsule, center=<-3.5311 0 0>, rotation=<rot -90 0 0>}]
}
constraints [
 {
  bone1Name=base_ear_jewel_l
  bone2Name=head
  swing1LimitDegrees=45
  swing1Motion=Free
  swing2LimitDegrees=45
  swing2Motion=Free
  twistLimitDegrees=45
  twistMotion=Free
 }
 {bone1Name=dyn_ear_jewel_l, bone2Name=base_ear_jewel_l, swing1LimitDegrees=45, swing2LimitDegrees=45, twistLimitDegrees=45}
 {
  bone1Name=base_ear_jewel_r
  bone2Name=head
  swing1LimitDegrees=45
  swing1Motion=Free
  swing2LimitDegrees=45
  swing2Motion=Free
  twistLimitDegrees=45
  twistMotion=Free
 }
 {bone1Name=dyn_ear_jewel_r, bone2Name=base_ear_jewel_r, swing1LimitDegrees=45, swing2LimitDegrees=45, twistLimitDegrees=45}
 {
  bone1Name=base_hair_short_bk_c
  bone2Name=head
  swing1LimitDegrees=45
  swing1Motion=Free
  swing2LimitDegrees=45
  swing2Motion=Free
  twistLimitDegrees=45
  twistMotion=Free
 }
 {bone1Name=dyn_hair_short_bk_c_01, bone2Name=base_hair_short_bk_c, swing1LimitDegrees=25, swing2LimitDegrees=35, twistLimitDegrees=5}
 {bone1Name=dyn_hair_short_bk_c_02, bone2Name=dyn_hair_short_bk_c_01, swing1LimitDegrees=25, swing2LimitDegrees=35, twistLimitDegrees=5}
 {
  bone1Name=dyn_hair_short_bk_r_02
  bone2Name=dyn_hair_short_bk_c_02
  swing1LimitDegrees=45
  swing1Motion=Free
  swing2LimitDegrees=45
  swing2Motion=Free
  twistLimitDegrees=45
  twistMotion=Free
 }
 {
  bone1Name=dyn_hair_short_bk_l_02
  bone2Name=dyn_hair_short_bk_c_02
  swing1LimitDegrees=45
  swing1Motion=Free
  swing2LimitDegrees=45
  swing2Motion=Free
  twistLimitDegrees=45
  twistMotion=Free
 }
 {bone1Name=dyn_hair_short_bk_c_03, bone2Name=dyn_hair_short_bk_c_02, swing1LimitDegrees=25, swing2LimitDegrees=35, twistLimitDegrees=5}
 {
  bone1Name=base_hair_short_bk_l
  bone2Name=head
  swing1LimitDegrees=45
  swing1Motion=Free
  swing2LimitDegrees=45
  swing2Motion=Free
  twistLimitDegrees=45
  twistMotion=Free
 }
 {bone1Name=dyn_hair_short_bk_l_01, bone2Name=base_hair_short_bk_l, swing1LimitDegrees=25, swing2LimitDegrees=35, twistLimitDegrees=5}
 {bone1Name=dyn_hair_short_bk_l_02, bone2Name=dyn_hair_short_bk_l_01, swing1LimitDegrees=25, swing2LimitDegrees=35, twistLimitDegrees=5}
 {
  bone1Name=dyn_hair_side_l_2
  bone2Name=dyn_hair_short_bk_l_02
  swing1LimitDegrees=45
  swing1Motion=Free
  swing2LimitDegrees=45
  swing2Motion=Free
  twistLimitDegrees=45
  twistMotion=Free
 }
 {bone1Name=dyn_hair_short_bk_l_03, bone2Name=dyn_hair_short_bk_l_02, swing1LimitDegrees=25, swing2LimitDegrees=35, twistLimitDegrees=5}
 {
  bone1Name=base_hair_short_bk_r
  bone2Name=head
  swing1LimitDegrees=45
  swing1Motion=Free
  swing2LimitDegrees=45
  swing2Motion=Free
  twistLimitDegrees=45
  twistMotion=Free
 }
 {bone1Name=dyn_hair_short_bk_r_01, bone2Name=base_hair_short_bk_r, swing1LimitDegrees=25, swing2LimitDegrees=35, twistLimitDegrees=5}
 {bone1Name=dyn_hair_short_bk_r_02, bone2Name=dyn_hair_short_bk_r_01, swing1LimitDegrees=25, swing2LimitDegrees=35, twistLimitDegrees=5}
 {
  bone1Name=dyn_hair_side_r_2
  bone2Name=dyn_hair_short_bk_r_02
  swing1LimitDegrees=45
  swing1Motion=Free
  swing2LimitDegrees=45
  swing2Motion=Free
  twistLimitDegrees=45
  twistMotion=Free
 }
 {bone1Name=dyn_hair_short_bk_r_03, bone2Name=dyn_hair_short_bk_r_02, swing1LimitDegrees=25, swing2LimitDegrees=35, twistLimitDegrees=5}
 {
  bone1Name=base_hair_side_l
  bone2Name=head
  swing1LimitDegrees=45
  swing1Motion=Free
  swing2LimitDegrees=45
  swing2Motion=Free
  twistLimitDegrees=45
  twistMotion=Free
 }
 {bone1Name=dyn_hair_side_l_1, bone2Name=base_hair_side_l, swing1LimitDegrees=45, swing2LimitDegrees=45, twistLimitDegrees=5}
 {bone1Name=dyn_hair_side_l_2, bone2Name=dyn_hair_side_l_1, swing1LimitDegrees=25, swing2LimitDegrees=25, twistLimitDegrees=5}
 {bone1Name=base_hair_side_r, bone2Name=head, swing1LimitDegrees=25, swing2LimitDegrees=25, twistLimitDegrees=10}
 {bone1Name=dyn_hair_side_r_1, bone2Name=base_hair_side_r, swing1LimitDegrees=45, swing2LimitDegrees=45, twistLimitDegrees=5}
 {bone1Name=dyn_hair_side_r_2, bone2Name=dyn_hair_side_r_1, swing1LimitDegrees=25, swing2LimitDegrees=45, twistLimitDegrees=5}
 {
  bone1Name=dyn_hair_side_l_2
  bone2Name=dyn_ear_jewel_l
  swing1LimitDegrees=45
  swing1Motion=Free
  swing2LimitDegrees=45
  swing2Motion=Free
  twistLimitDegrees=45
  twistMotion=Free
 }
]
