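# Post-animation Blueprint for the SableHorizon female hair face accessory, adding secondary motion to hair and chain bones.
# Copied parent pose, with shoulders from the body mesh, feeds a Control Rig.
# Backbling or capes switch back hair from head-space RigidBody simulation with parameter-driven gravity to bone constraints.
# AnimDynamics then springs front chain and hair, with separate front-end and gameplay setups, before additive side-hair rotations.
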AnimBlueprint /BRCosmetics/Characters/Player/Female/Medium/Bodies/F_MED_SableHorizon_Hair/Meshes/Parts/F_MED_SableHorizon_Hair_FaceAcc_AnimBP
target_skeleton=&/F_MED_SableHorizon_Hair_FaceAcc_Skeleton
graphs [
 AnimGraph {nodes [
   Copy Pose From Mesh() {bUseAttachedParent=true}
   n1 = Anim Dynamics - Bone: dyn_chain_fr!(ComponentPose=Local To Component(LocalPose=Use cached pose 'RBAN_cache'()), bAlphaBoolEnabled:bool=True, ExternalForce="0,0,0") -> (Pose) {
    alphaInputType=Bool
    angularDampingOverride=1
    angularSpringConstant=10000
    bAngularSpring=true
    bOverrideAngularDamping=true
    bUseGravityOverride=true
    gravityScale=0.35
    simulationSpace=Actor
    alphaBoolBlend {blendInTime=0.2, blendOption=Linear, blendOutTime=0.2}
    boundBone {boneName=dyn_chain_fr}
    componentLinearAccScale=<1 1 1>
    componentLinearVelScale=<1 1 1>
    gravityOverride {z=-980}
    physicsBodyDefinitions [{
      boundBone {boneName=dyn_chain_fr}
      boxExtents=<20 20 20>
      constraintSetup {angularLimitsMax=<5 0 10>, angularLimitsMin=<-5 -25 -10>, angularTarget=<1 0 0>}
      localJointOffset=<3 0 0>
     }]
   }
   n2 = Anim Dynamics - Bone: dyn_hair_fr_l!(ComponentPose=@n1, bAlphaBoolEnabled:bool=True, ExternalForce="0,0,0") -> (Pose) {
    alphaInputType=Bool
    angularDampingOverride=1
    angularSpringConstant=10000
    bAngularSpring=true
    bOverrideAngularDamping=true
    bUseGravityOverride=true
    gravityScale=0.35
    simulationSpace=RootRelative
    alphaBoolBlend {blendInTime=0.2, blendOption=Linear, blendOutTime=0.2}
    boundBone {boneName=dyn_hair_fr_l}
    componentLinearAccScale=<1 1 1>
    componentLinearVelScale=<1 1 1>
    gravityOverride {z=-980}
    physicsBodyDefinitions [{
      boundBone {boneName=dyn_hair_fr_l}
      boxExtents=<20 20 20>
      constraintSetup {angularLimitsMax=<10 0 10>, angularLimitsMin=<-10 -10 -10>, angularTarget=<1 0 0>}
      localJointOffset=<3 0 0>
     }]
   }
   n3 = Anim Dynamics - Bone: dyn_hair_fr_r!(ComponentPose=@n2, bAlphaBoolEnabled:bool=True, ExternalForce="0,0,0") -> (Pose) {
    alphaInputType=Bool
    angularDampingOverride=1
    angularSpringConstant=10000
    bAngularSpring=true
    bOverrideAngularDamping=true
    bUseGravityOverride=true
    gravityScale=0.35
    simulationSpace=RootRelative
    alphaBoolBlend {blendInTime=0.2, blendOption=Linear, blendOutTime=0.2}
    boundBone {boneName=dyn_hair_fr_r}
    componentLinearAccScale=<1 1 1>
    componentLinearVelScale=<1 1 1>
    gravityOverride {z=-980}
    physicsBodyDefinitions [{
      boundBone {boneName=dyn_hair_fr_r}
      boxExtents=<20 20 20>
      constraintSetup {angularLimitsMax=<10 0 10>, angularLimitsMin=<-10 -10 -10>, angularTarget=<1 0 0>}
      localJointOffset=<3 0 0>
     }]
   }
   n4 = Component To Local(ComponentPose=@n3)
   n5 = Anim Dynamics - Bone: dyn_chain_fr(ComponentPose=Local To Component(LocalPose=Use cached pose 'RBAN_cache'()), bAlphaBoolEnabled=True, ExternalForce="0,0,0") {
    alphaInputType=Bool
    angularDampingOverride=1
    angularSpringConstant=10000
    bAngularSpring=true
    bOverrideAngularDamping=true
    bUseGravityOverride=true
    gravityScale=0.35
    simulationSpace=Actor
    alphaBoolBlend {blendInTime=0.2, blendOption=Linear, blendOutTime=0.2}
    boundBone {boneName=dyn_chain_fr}
    componentLinearAccScale=<1 1 1>
    componentLinearVelScale=<1 1 1>
    gravityOverride {z=-980}
    physicsBodyDefinitions [{
      boundBone {boneName=dyn_chain_fr}
      boxExtents=<20 20 20>
      constraintSetup {angularLimitsMax=<5 0 10>, angularLimitsMin=<-5 -25 -10>, angularTarget=<1 0 0>}
      localJointOffset=<3 0 0>
     }]
    simSpaceSettings {maxAngularAcceleration=35, maxAngularVelocity=7, simSpaceAngularAlpha=1}
   }
   n6 = Anim Dynamics - Bone: dyn_hair_fr_l(ComponentPose=@n5, bAlphaBoolEnabled=True, ExternalForce="0,0,0") {
    alphaInputType=Bool
    angularDampingOverride=1
    angularSpringConstant=10000
    bAngularSpring=true
    bOverrideAngularDamping=true
    bUseGravityOverride=true
    gravityScale=0.35
    simulationSpace=RootRelative
    alphaBoolBlend {blendInTime=0.2, blendOption=Linear, blendOutTime=0.2}
    boundBone {boneName=dyn_hair_fr_l}
    componentLinearAccScale=<1 1 1>
    componentLinearVelScale=<1 1 1>
    gravityOverride {z=-980}
    physicsBodyDefinitions [{
      boundBone {boneName=dyn_hair_fr_l}
      boxExtents=<20 20 20>
      constraintSetup {angularLimitsMax=<10 0 10>, angularLimitsMin=<-10 -10 -10>, angularTarget=<1 0 0>}
      localJointOffset=<3 0 0>
     }]
    simSpaceSettings {maxAngularAcceleration=35, maxAngularVelocity=7, simSpaceAngularAlpha=0.7}
   }
   n7 = Anim Dynamics - Bone: dyn_hair_fr_r(ComponentPose=@n6, bAlphaBoolEnabled=True, ExternalForce="0,0,0") {
    alphaInputType=Bool
    angularDampingOverride=1
    angularSpringConstant=10000
    bAngularSpring=true
    bOverrideAngularDamping=true
    bUseGravityOverride=true
    gravityScale=0.35
    simulationSpace=RootRelative
    alphaBoolBlend {blendInTime=0.2, blendOption=Linear, blendOutTime=0.2}
    boundBone {boneName=dyn_hair_fr_r}
    componentLinearAccScale=<1 1 1>
    componentLinearVelScale=<1 1 1>
    gravityOverride {z=-980}
    physicsBodyDefinitions [{
      boundBone {boneName=dyn_hair_fr_r}
      boxExtents=<20 20 20>
      constraintSetup {angularLimitsMax=<10 0 10>, angularLimitsMin=<-10 -10 -10>, angularTarget=<1 0 0>}
      localJointOffset=<3 0 0>
     }]
    simSpaceSettings {maxAngularAcceleration=35, maxAngularVelocity=7, simSpaceAngularAlpha=0.7}
   }
   n8 = Component To Local(ComponentPose=@n7)
   n9 = Get bIsInFrontEnd()
   n10 = "Transform (Modify) Bone - Bone: base_hair_side_l"!(
    ComponentPose=Local To Component(LocalPose=Blend Poses by bool(BlendPose_0=@n8, BlendPose_1=@n4, bActiveValue=@n9))
    bAlphaBoolEnabled:bool=True
    Translation="0,0,0"
    Rotation="-15,0,0"
    Scale="1,1,1"
   ) -> (Pose) {
    alphaInputType=Bool
    rotationMode=BMM_Additive
    rotationSpace=BCS_BoneSpace
    alphaBoolBlend {blendInTime=0.2, blendOutTime=0.2}
    boneToModify {boneName=base_hair_side_l}
    rotation {pitch=-15}
   }
   n11 = "Transform (Modify) Bone - Bone: base_hair_side_r"!(ComponentPose=@n10, bAlphaBoolEnabled:bool=True, Translation="0,0,0", Rotation="-10,0,0", Scale="1,1,1") -> (Pose) {
    alphaInputType=Bool
    rotationMode=BMM_Additive
    rotationSpace=BCS_BoneSpace
    alphaBoolBlend {blendInTime=0.2, blendOutTime=0.2}
    boneToModify {boneName=base_hair_side_r}
    rotation {pitch=-10}
   }
   Output Pose(Result=Component To Local(ComponentPose=@n11)) {layerGroup=None}
   n12 = Layered blend per bone(
    BasePose=Copy Pose From Mesh() {bCopyCurves=true, bUseAttachedParent=true}
    BlendPoses_0=Copy Pose From Mesh(SourceMeshComponent=Get Body_Mesh())
   ) {layerSetup [{branchFilters [
       {boneName=clavicle_l}
       {boneName=clavicle_r}
       {boneName=upperarm_l}
       {boneName=upperarm_r}
       {boneName=upperarm_twist_01_l}
       {boneName=upperarm_twist_01_r}
       {boneName=deform_upperarm_bk_base_l}
       {boneName=deform_upperarm_bk_l}
       {boneName=deform_upperarm_fr_base_l}
       {boneName=deform_upperarm_fr_l}
       {boneName=deform_upperarm_up_base_l}
       {boneName=deform_upperarm_up_l}
       {boneName=deform_upperarm_bk_base_r}
       {boneName=deform_upperarm_bk_r}
       {boneName=deform_upperarm_fr_base_r}
       {boneName=deform_upperarm_fr_r}
       {boneName=deform_upperarm_up_base_r}
       {boneName=deform_upperarm_up_r}
      ]}]}
   n13 = Control Rig(Source=@n12) {controlRigAssetReference {blueprintRigClass=&/F_MED_SableHorizon_Hair_FaceAcc_CtrlRig.F_MED_SableHorizon_Hair_FaceAcc_CtrlRig_C}}
   Save cached pose 'CtrRig_cache'(Pose=@n13) {cacheName=CtrRig_cache}
   n14 = "Constraint - dyn_hair_short_bk_r_01 (PP)"!(ComponentPose=Local To Component(LocalPose=Use cached pose 'CtrRig_cache'()), Alpha=1, ConstraintWeights_0=1, ConstraintWeights_1:float=0) -> (Pose) {
    alphaBoolBlend {blendOption=Linear}
    boneToModify {boneName=dyn_hair_short_bk_r_01}
    constraintSetup [{targetBone {boneName=head}}, {targetBone {boneName=neck_02}}]
    constraintWeights [1,1]
   }
   n15 = "Constraint - dyn_hair_short_bk_r_02 (PPP)"!(ComponentPose=@n14, Alpha=1, ConstraintWeights_0=0, ConstraintWeights_1:float=0.5, ConstraintWeights_2:float=0.5) -> (Pose) {
    alphaBoolBlend {blendOption=Linear}
    boneToModify {boneName=dyn_hair_short_bk_r_02}
    constraintSetup [{targetBone {boneName=clavicle_r}}, {targetBone {boneName=neck_02}}, {targetBone {boneName=neck_01}}]
    constraintWeights [1,1,1]
   }
   n16 = "Constraint - dyn_hair_short_bk_r_02 (PPP)"(ComponentPose=@n15, Alpha=1, ConstraintWeights_0=0.5, ConstraintWeights_1=0, ConstraintWeights_2=0.5) {
    alphaBoolBlend {blendOption=Linear}
    boneToModify {boneName=dyn_hair_short_bk_r_02}
    constraintSetup [{targetBone {boneName=clavicle_r}}, {targetBone {boneName=neck_02}}, {targetBone {boneName=neck_01}}]
    constraintWeights [1,1,1]
   }
   n17 = "Constraint - dyn_hair_short_bk_l_01 (PP)"!(ComponentPose=@n16, Alpha=1, ConstraintWeights_0=1, ConstraintWeights_1:float=0) -> (Pose) {
    alphaBoolBlend {blendOption=Linear}
    boneToModify {boneName=dyn_hair_short_bk_l_01}
    constraintSetup [{targetBone {boneName=head}}, {targetBone {boneName=neck_02}}]
    constraintWeights [1,1]
   }
   n18 = "Constraint - dyn_hair_short_bk_l_02 (PPP)"!(ComponentPose=@n17, Alpha=1, ConstraintWeights_0=0, ConstraintWeights_1:float=0.5, ConstraintWeights_2:float=0.5) -> (Pose) {
    alphaBoolBlend {blendOption=Linear}
    boneToModify {boneName=dyn_hair_short_bk_l_02}
    constraintSetup [{targetBone {boneName=clavicle_r}}, {targetBone {boneName=neck_02}}, {targetBone {boneName=neck_01}}]
    constraintWeights [1,1,1]
   }
   n19 = "Constraint - dyn_hair_short_bk_l_03 (PPP)"!(ComponentPose=@n18, Alpha=1, ConstraintWeights_0=0.5, ConstraintWeights_1:float=0, ConstraintWeights_2:float=0.5) -> (Pose) {
    alphaBoolBlend {blendOption=Linear}
    boneToModify {boneName=dyn_hair_short_bk_l_03}
    constraintSetup [{targetBone {boneName=clavicle_l}}, {targetBone {boneName=neck_02}}, {targetBone {boneName=neck_01}}]
    constraintWeights [1,1,1]
   }
   n20 = "Constraint - dyn_hair_short_bk_c_01 (PP)"!(ComponentPose=@n19, Alpha=1, ConstraintWeights_0=1, ConstraintWeights_1:float=0) -> (Pose) {
    alphaBoolBlend {blendOption=Linear}
    boneToModify {boneName=dyn_hair_short_bk_c_01}
    constraintSetup [{targetBone {boneName=head}}, {targetBone {boneName=neck_02}}]
    constraintWeights [1,1]
   }
   n21 = "Constraint - dyn_hair_short_bk_c_02 (PPP)"!(ComponentPose=@n20, Alpha=1, ConstraintWeights_0=0, ConstraintWeights_1:float=0.5, ConstraintWeights_2:float=0.5) -> (Pose) {
    alphaBoolBlend {blendOption=Linear}
    boneToModify {boneName=dyn_hair_short_bk_c_02}
    constraintSetup [{targetBone {boneName=clavicle_r}}, {targetBone {boneName=neck_02}}, {targetBone {boneName=neck_01}}]
    constraintWeights [1,1,1]
   }
   n22 = "Constraint - dyn_hair_short_bk_c_03 (PPP)"!(ComponentPose=@n21, Alpha=1, ConstraintWeights_0=0.5, ConstraintWeights_1:float=0, ConstraintWeights_2:float=0.5) -> (Pose) {
    alphaBoolBlend {blendOption=Linear}
    boneToModify {boneName=dyn_hair_short_bk_c_03}
    constraintSetup [{targetBone {boneName=clavicle_l}}, {targetBone {boneName=spine_05}}, {targetBone {boneName=neck_01}}]
    constraintWeights [1,1,1]
   }
   n23 = Component To Local(ComponentPose=@n22)
   n24 = RigidBody!(
    ComponentPose=Local To Component(LocalPose=Use cached pose 'CtrRig_cache'())
    Alpha
    OverrideWorldGravity:Vector=Get Gravity Override()
    ExternalForce
   ) -> (Pose) {
    bClampLinearTranslationLimitToRefPose=true
    bOverrideWorldGravity=true
    overridePhysicsAsset=&/F_MED_SableHorizon_Hair_FaceAcc_Physics
    simulationSpace=BaseBoneSpace
    alphaBoolBlend {blendOption=Linear}
    baseBoneRef {boneName=head}
    componentAppliedLinearAccClamp=<0 0 0>
    overrideWorldGravity {z=-980}
    simSpaceSettings {
     dampingAlpha=0
     maxAngularAcceleration=40
     maxAngularVelocity=10
     maxLinearAcceleration=1200
     maxLinearVelocity=800
     velocityScaleZ=0.5
     worldAlpha=1
     externalLinearDragV=<0.3 0.3 0.3>
    }
   }
   n25 = Component To Local(ComponentPose=@n24)
   n26 = OR Boolean(A=Get isWearingBackbling(), B=Get Is Wearing Cape() -> (Is Wearing Cape:bool))
   n27 = "Transform (Modify) Bone - Bone: base_chain_fr"(
    ComponentPose=Local To Component(LocalPose=Blend Poses by bool(BlendPose_0=@n23, BlendPose_1=@n25, bActiveValue=@n26) {blendType=Linear})
    Alpha=1
    Translation="0,0,0"
    Rotation="2,0,0"
    Scale="1,1,1"
   ) {
    rotationMode=BMM_Additive
    rotationSpace=BCS_BoneSpace
    translationMode=BMM_Additive
    translationSpace=BCS_BoneSpace
    boneToModify {boneName=base_chain_fr}
    rotation {pitch=2}
   }
   Save cached pose 'RBAN_cache'(Pose=Component To Local(ComponentPose=@n27)) {cacheName=RBAN_cache}
  ]}
 EventGraph {nodes [
   Event BlueprintUpdateAnimation()
   > n1 = Sequence()
   > n2 = "Gravity Override State Function (Parameter Asset) (Nativized)"(
    CharacterPartAnimInstance=Self-Reference()
    GameTimeInSeconds=GetGameTimeInSeconds()
    Parameters=&/F_MED_SableHorizon_Hair_FaceAcc_Parameters
   )
   > Set Gravity Override(Gravity Override=@n2)
   > Set RBAN Settings Out(RBAN Settings Out=@n2.OutRigidBodyAnimNodeParameters)
   n3 = CheckForBackbling()
   Set isWearingBackbling(
    execute=@n1.then_1
    isWearingBackbling=AND Boolean!(A=@n3, B=NOT Boolean(A=@n3.bIsBackblingACape), C:bool=NOT Boolean(A=@n3.bIsBackblingLong)) -> (ReturnValue)
   )
   > Set Is Wearing Cape(Is Wearing Cape=OR Boolean(A=@n3.bIsBackblingACape, B=@n3.bIsBackblingLong))
   n4 = Event OnCharacterCustomizationLoaded()
   > Set Body_Mesh(Body_Mesh=@n4.CustomBaseMesh)
  ]}
]
variables [
 Body_Mesh {type=SkeletalMeshComponent}
 Gravity Override {type=Vector}
 Is Wearing Cape {type=boolean}
 isWearingBackbling {type=boolean}
 RBAN Settings Out {type=FortRigidBodyAnimNodeParameters}
]
